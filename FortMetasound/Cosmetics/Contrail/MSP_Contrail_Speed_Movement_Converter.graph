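# MetaSound patch for contrail cosmetics that converts a speed value into frequency, frequency gain and volume for movement-driven sound.
# Three float MapRange nodes remap Speed from a fixed input range to caller-supplied min and max values for each output.
# The volume multiplier also scales the incoming left and right audio directly, while frequency and frequency gain are only output as values, with no filtering inside the patch.
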MetaSoundPatch /FortMetasound/Cosmetics/Contrail/MSP_Contrail_Speed_Movement_Converter
inputs [
 Frequency At Max Speed {type=Float}
 Frequency At Min Speed {type=Float}
 "Gain  Multiplier aA Max Speed" {type=Float}
 Gain Multiplier At Min Speed {type=Float}
 In Left {type=Audio}
 In Right {type=Audio}
 Max Freq Gain {type=Float}
 Min Freq Gain {type=Float}
 Speed {type=Float}
]
outputs [
 Freq Gain {type=Float}
 Frequency {type=Float}
 Out Left {type=Audio}
 Out Right {type=Audio}
 Volume {type=Float}
]
pages [Default {
  input_defaults {
   Frequency At Max Speed=0
   Frequency At Min Speed=0
   "Gain  Multiplier aA Max Speed"=1
   Gain Multiplier At Min Speed=1
   In Left=null
   In Right=null
   Max Freq Gain=2
   Min Freq Gain=1
   Speed=0
  }
  nodes [
   MapRangeMapRange_0 = MapRange.MapRange:Float(Out Range B=^Max Freq Gain, Out Range A=^Min Freq Gain, In=UE.Reroute(Value=^Speed).Value) {also {defaults {In Range A=900, In Range B=6000}}}
   MapRangeMapRange_2 = MapRange.MapRange:Float(Out Range B=^Frequency At Max Speed, Out Range A=^Frequency At Min Speed, In=UE.Reroute(Value=^Speed).Value) {also {defaults {In Range A=900, In Range B=6000}}}
   MapRangeMapRange_1 = MapRange.MapRange:Float(Out Range B=^"Gain  Multiplier aA Max Speed", Out Range A=^Gain Multiplier At Min Speed, In=UE.Reroute(Value=^Speed).Value) {also {defaults {In Range A=900, In Range B=6000}}}
   Multiply_0 = UE.Multiply:Audio by Float(PrimaryOperand=^In Left, AdditionalOperands=@MapRangeMapRange_1.Out Value)
   Multiply_1 = UE.Multiply:Audio by Float(PrimaryOperand=^In Right, AdditionalOperands=@MapRangeMapRange_1.Out Value)
  ]
  links [
   {dst="@Freq Gain", src=MapRangeMapRange_0.Out Value}
   {dst="@Volume", src=MapRangeMapRange_1.Out Value}
   {dst="@Frequency", src=MapRangeMapRange_2.Out Value}
   {dst="@Out Left", src=Multiply_0.Out}
   {dst="@Out Right", src=Multiply_1.Out}
  ]
 }]
